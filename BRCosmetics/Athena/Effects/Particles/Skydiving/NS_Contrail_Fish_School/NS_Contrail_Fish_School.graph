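# Skydiving contrail cosmetic: a school of mesh fish orbiting the player, with bubbles and ribbon trails.
# Fish001 bursts SpawnCount fish around the owner with sine bob and curl noise, slerping facing from velocity to skydive rotation past a dive threshold.
# The ribbon emitter follows Fish001 via particle attribute reads, while CPU and GPU bubble emitters blend vortex swirl into stronger lift when diving, with front-end overrides.
# Five older or debug emitters are disabled.
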NiagaraSystem /BRCosmetics/Athena/Effects/Particles/Skydiving/NS_Contrail_Fish_School/NS_Contrail_Fish_School
dependencies {
 usedDataInterfaces [&NiagaraDataInterfaceCurve, &NiagaraDataInterfaceParticleRead]
 usedDynamicInputs []
 usedModules [&SystemState, /Niagara/Modules/System/SystemLifeCycle]
}
emitters {
 DEBUG {
  bEnabled=false
  emitterUpdateScript [
   &EmitterScalability(
    Module.bApplySpawnCountScale=false
    Module.bApplySpawnCountScaleByDistanceFraction=false
    Module.bUseMaxDistance=false
    Module.bUseMinDistance=false
   ) {enabled=false}
   &EmitterLifeCycle() {enabled=false}
   &EmitterState()
   &SpawnBurst_Instantaneous(Spawn Count=10, Use Spawn Probability=false) {enabled=false}
   &SpawnRate(SpawnRate=17, Use Spawn Probability=false)
  ]
  particleSpawnScript [
   /Game/Effects/Niagara/Modules/Orbit/FN_Orbit_Init_Variables_v2() {enabled=false}
   /SetVariables(Particles.fish_scale2:Vector3f=<5 5 5>) {isSetParameters=true}
   /Game/Effects/Niagara/Modules/Groups/FN_SetSpawnGroup_v1/FN_SetSpawnGroup(Override SpawnGroup=true, SpawnGroup=4) {enabled=false}
   /Niagara/Modules/Spawn/Initialization/InitializeParticle(
    Lifetime=4
    Mesh Scale=&VectorFromFloat(Value=&UniformRangedFloat(Minimum=1, Maximum=1))
    Module.Write Color=true
    Module.Write Lifetime=true
    Module.Write Mass=true
    Module.Write Material Random=true
    Module.Write Position=false
    Module.Write Scale=true
    Module.Write SpriteRotation=false
    Module.Write SpriteSize=false
    Position=&SimulationPosition()
   )
   &AddRotationalVelocity(
    Coordinate Space:ENiagaraMeshTransforms=::NewEnumerator1
    Rotation Rate=&UniformRangedVector(Minimum=<-0.25 0 0>, Maximum=<0.25 0 0>)
   ) {enabled=false}
   &ApplyInitialForces() {enabled=false}
   /Niagara/Modules/Spawn/Orientation/V2/InitialMeshOrientation(
    Orientation Coordinate Space:ENiagaraCoordinateSpace=::World
    Orientation Vector=<0 0 -1>
    Rotation Coordinate Space:ENiagaraMeshTransforms=::NewEnumerator1
    Use Orientation Vector=true
    Use Rotation Vector=true
   ) {enabled=false}
   &SystemLocation(MaskSpawnGroup=false, Offset=<0 0 155>)
   &DynamicMaterialParameters(
    Index 0 Param 1=&UniformRangedFloat(Minimum=0.1, Maximum=1.2)
    Index 0 Param 2=&UniformRangedFloat(Minimum=0.5, Maximum=1.3)
    Module.Param0WriteEnabled=true
    Module.Param1WriteEnabled=true
    Module.Param2WriteEnabled=true
    Module.Param3WriteEnabled=true
   ) {enabled=false}
   &CylinderLocation(
    Cylinder Height=10
    Cylinder Radius=50
    Lathe Profile=DataInterface() {propertyValues {Curve {keys [<at 0 0.451 RCIM_Cubic>, {arriveTangent=0.049, interpMode=RCIM_Cubic, leaveTangent=0.049, time=0.5, value=1}, <at 1 0.5 RCIM_Cubic>]}}}
    MaskSpawnGroup=false
    Module.Enable Lathe Profile=false
    Module.Override Randomness=false
    Module.Surface Only=true
    Surface Only Band Thickness=30
    Use Endcaps In Surface Only Mode=false
   )
  ]
  particleUpdateScript [
   /Game/Effects/Niagara/Modules/Skydiving/FN_DivingAttributes/FN_DivingAttributes_v2(SkeletalMesh:NiagaraDataInterfaceSkeletalMesh=$User.SkeletalMesh) {enabled=false}
   /Game/Effects/Niagara/Modules/Orbit/FN_Orbit_v3(
    Height min/max:Vector2f=<0 25>
    Radius min/max:Vector2f=<25 100>
    Rotation Offset min/max:Vector2f=<1 5>
    Rotation Rate min/max:Vector2f=<0.1 0.2>
   ) {enabled=false}
   /SetVariables() {isSetParameters=true}
   &AddVelocity(Velocity=&Lerp_Vector(A=<0 0 0>, B=<0 0 -1>, Alpha=$Particles.Dive)) {enabled=false}
   /Niagara/Modules/Update/Velocity/VortexVelocity(Module.Use Influence Falloff=false)
   &SolveForcesAndVelocity(Clamp Velocity=false, Limit Acceleration=false)
   &ParticleState()
   /Niagara/Modules/Update/Orientation/V2/UpdateMeshOrientation(Rotation Rate=0, Rotation Vector:Vector3f=<0 1 0>) {enabled=false}
   &Color() {enabled=false}
   &ScaleMeshSize(Scale Factor=&VectorFromFloat(Value=&FloatFromCurve(FloatCurve=DataInterface() {propertyValues {Curve {keys [<at -0.0046 0>,<at 0.5 1>,<at 1.0011 0.002>]}}}))) {enabled=false}
   &CalculateAccurateVelocity() {enabled=false}
  ]
  rendererClasses [&NiagaraMeshRendererProperties]
  renderers [
   {rendererClass=&NiagaraMeshRendererProperties, rendererIndex=0}
   {rendererClass=&NiagaraMeshRendererProperties, rendererIndex=1}
  ]
 }
 FN_Contrails_Bubble_Low {
  emitterSpawnScript [&FN_PreserveSkydiveUserVariables() {module=PreserveUserVariables}]
  emitterUpdateScript [
   &EmitterState()
   &SpawnRate(SpawnRate=50, Use Spawn Probability=false)
  ]
  particleSpawnScript [
   &FN_DivingAttributes_v2(SkeletalMesh:NiagaraDataInterfaceSkeletalMesh=$User.SkeletalMesh)
   &InitializeParticle(
    Lifetime=&UniformRangedFloat(Minimum=0.7, Maximum=1.5)
    Module.Write Color=true
    Module.Write Lifetime=true
    Module.Write Mass=true
    Module.Write Material Random=true
    Module.Write Position=false
    Module.Write Scale=false
    Module.Write SpriteRotation=true
    Module.Write SpriteSize=true
    Position=&SimulationPosition()
    Sprite Size=&Vector2DFromFloat(Value=&UniformRangedFloat(Minimum=4, Maximum=6))
   )
   &CylinderLocation(
    Cylinder Height=10
    Cylinder Radius=&Lerp_Float(A=50, B=45, Alpha=$Particles.Dive)
    Lathe Profile=DataInterface() {propertyValues {Curve {keys [<at 0 0.451 RCIM_Cubic>, {arriveTangent=0.049, interpMode=RCIM_Cubic, leaveTangent=0.049, time=0.5, value=1}, <at 1 0.5 RCIM_Cubic>]}}}
    MaskSpawnGroup=false
    Module.Enable Lathe Profile=false
    Module.Override Randomness=false
    Module.Surface Only=true
    Offset=&Lerp_Vector(A=<0 0 125>, B=<0 0 270>, Alpha=$Particles.Dive)
    Surface Only Band Thickness=30
    Use Endcaps In Surface Only Mode=false
   )
  ]
  particleUpdateScript [
   &FN_DivingAttributes_v2(SkeletalMesh:NiagaraDataInterfaceSkeletalMesh=$User.SkeletalMesh) {module=FN_DivingAttributes_v2001}
   &VortexVelocity(Module.Use Influence Falloff=false, Velocity Amount:NiagaraFloat=&Lerp_Float(A=200, B=0, Alpha=$Particles.Dive))
   &ParticleState()
   &GravityForce(Gravity=&Lerp_Vector(A=<0 0 1000>, B=<0 1 5000>, Alpha=$Particles.Dive))
   &GravityForce(
    Coordinate Space=::Local
    Gravity=&MakeVector(
     X=0
     Y=&FN_FrontEndFloatModifier(
      Gameplay=0
      Front End=&Lerp_Float(
       A=0
       B=-5000
       Alpha=/Game/Effects/Niagara/DynamicInputs/Math/FN_Fit(
        Clamp:NiagaraBool=true
        Value:NiagaraFloat=$Particles.Dive
        SourceMin:NiagaraFloat=0
        SourceMax:NiagaraFloat=0.1
        DestMin:NiagaraFloat=0
        DestMax:NiagaraFloat=1
       )
      )
     )
     Z=0
    )
   ) {module=GravityForce001}
   &Drag(Drag=5, Use Linear Drag=true, Use Rotational Drag=false)
   &SolveForcesAndVelocity(Clamp Velocity=false, Limit Acceleration=false)
   &ScaleSpriteSize(Scale Factor=&Vector2DFromFloat(Value=&FloatFromCurve(FloatCurve=DataInterface() {propertyValues {Curve {keys [
          {arriveTangent=4.4544, interpMode=RCIM_Cubic, leaveTangent=4.4544, tangentMode=RCTM_User, time=0, value=0}
          <at 0.4741 0.9412 RCIM_Cubic>
          {arriveTangent=-2.1034, interpMode=RCIM_Cubic, leaveTangent=-2.1034, tangentMode=RCTM_User, time=1, value=0}
         ]}}})))
  ]
  rendererClasses [&NiagaraSpriteRendererProperties]
  renderers [{rendererClass=&NiagaraSpriteRendererProperties}]
 }
 FN_Contrails_Bubbles {
  simTarget=GPUComputeSim
  emitterSpawnScript [&FN_PreserveSkydiveUserVariables() {module=PreserveUserVariables}]
  emitterUpdateScript [
   &EmitterState()
   &SpawnRate(SpawnRate=200, Use Spawn Probability=false)
  ]
  particleSpawnScript [
   &FN_DivingAttributes_v2(SkeletalMesh:NiagaraDataInterfaceSkeletalMesh=$User.SkeletalMesh)
   &InitializeParticle(
    Lifetime=&UniformRangedFloat(Minimum=0.7, Maximum=1.5)
    Module.Write Color=true
    Module.Write Lifetime=true
    Module.Write Mass=true
    Module.Write Material Random=true
    Module.Write Position=false
    Module.Write Scale=false
    Module.Write SpriteRotation=true
    Module.Write SpriteSize=true
    Position=&SimulationPosition()
    Sprite Size=&Vector2DFromFloat(Value=&UniformRangedFloat(Minimum=4, Maximum=6))
   )
   &CylinderLocation(
    Cylinder Height=10
    Cylinder Radius=&Lerp_Float(A=50, B=45, Alpha=$Particles.Dive)
    Lathe Profile=DataInterface() {propertyValues {Curve {keys [<at 0 0.451 RCIM_Cubic>, {arriveTangent=0.049, interpMode=RCIM_Cubic, leaveTangent=0.049, time=0.5, value=1}, <at 1 0.5 RCIM_Cubic>]}}}
    MaskSpawnGroup=false
    Module.Enable Lathe Profile=false
    Module.Override Randomness=false
    Module.Surface Only=true
    Offset=&Lerp_Vector(A=<0 0 125>, B=<0 0 270>, Alpha=$Particles.Dive)
    Surface Only Band Thickness=30
    Use Endcaps In Surface Only Mode=false
   )
  ]
  particleUpdateScript [
   &FN_DivingAttributes_v2(SkeletalMesh:NiagaraDataInterfaceSkeletalMesh=$User.SkeletalMesh) {module=FN_DivingAttributes_v2001}
   &VortexVelocity(Module.Use Influence Falloff=false, Velocity Amount:NiagaraFloat=&Lerp_Float(A=200, B=0, Alpha=$Particles.Dive))
   &ParticleState()
   &GravityForce(Gravity=&Lerp_Vector(A=<0 0 1000>, B=<0 1 5000>, Alpha=$Particles.Dive))
   &GravityForce(
    Coordinate Space=::Local
    Gravity=&MakeVector(
     X=0
     Y=&FN_FrontEndFloatModifier(
      Gameplay=0
      Front End=&Lerp_Float(
       A=0
       B=-5000
       Alpha=&FN_Fit(
        Clamp:NiagaraBool=true
        Value:NiagaraFloat=$Particles.Dive
        SourceMin:NiagaraFloat=0
        SourceMax:NiagaraFloat=0.1
        DestMin:NiagaraFloat=0
        DestMax:NiagaraFloat=1
       )
      )
     )
     Z=0
    )
   ) {module=GravityForce001}
   /Niagara/Modules/Update/Forces/CurlNoiseForce(Noise Strength=100, Pan Noise=false) {module=CurlNoiseForce001}
   &Drag(Drag=5, Use Linear Drag=true, Use Rotational Drag=false)
   &SolveForcesAndVelocity(Clamp Velocity=false, Limit Acceleration=false)
   &ScaleSpriteSize(Scale Factor=&Vector2DFromFloat(Value=&FloatFromCurve(FloatCurve=DataInterface() {propertyValues {Curve {keys [
          {arriveTangent=4.4544, interpMode=RCIM_Cubic, leaveTangent=4.4544, tangentMode=RCTM_User, time=0, value=0}
          <at 0.4741 0.9412 RCIM_Cubic>
          {arriveTangent=-2.1034, interpMode=RCIM_Cubic, leaveTangent=-2.1034, tangentMode=RCTM_User, time=1, value=0}
         ]}}})))
  ]
  rendererClasses [&NiagaraSpriteRendererProperties]
  renderers [{rendererClass=&NiagaraSpriteRendererProperties}]
 }
 FN_Contrails_Bubbles_Child {
  bEnabled=false
  emitterSpawnScript [&FN_PreserveSkydiveUserVariables() {module=PreserveUserVariables}]
  emitterUpdateScript [
   &EmitterState()
   &SpawnRate(SpawnRate=50, Use Spawn Probability=false) {enabled=false}
  ]
  particleSpawnScript [
   &FN_SetSpawnGroup(Override SpawnGroup=true, SpawnGroup=4) {enabled=false}
   &InitializeParticle(
    Lifetime=&UniformRangedFloat(Minimum=0.7, Maximum=1.5)
    Module.Write Color=true
    Module.Write Lifetime=true
    Module.Write Mass=true
    Module.Write Material Random=true
    Module.Write Position=false
    Module.Write Scale=false
    Module.Write SpriteRotation=true
    Module.Write SpriteSize=true
    Position=&SimulationPosition()
    Sprite Size=&Vector2DFromFloat(Value=&UniformRangedFloat(Minimum=3, Maximum=7))
   )
   &SampleSkeletalMeshSkeleton(Mesh:NiagaraDataInterfaceSkeletalMesh=$User.SkeletalMesh, Specific Bone Index:NiagaraInt32=$Particles.SpawnGroup) {enabled=false}
   /Niagara/Modules/Spawn/Location/SkeletalMeshSkeletonLocation(Inherit Velocity From Mesh=0.2, Module.Inherit Velocity=true, Module.Use Sampled Bone as Ribbon ID=true) {enabled=false}
  ]
  particleUpdateScript [
   &FN_DivingAttributes_v2(SkeletalMesh:NiagaraDataInterfaceSkeletalMesh=$User.SkeletalMesh)
   &ParticleState()
   &AddVelocity(Velocity=&MakeVector(X=0, Y=0, Z=&Lerp_Float(A=0, B=0, Alpha=$Particles.Dive))) {enabled=false}
   &GravityForce(Gravity=&Lerp_Vector(A=<0 0 150>, B=<1 1 2000>, Alpha=$Particles.Dive))
   &Drag(Drag=5, Use Linear Drag=true, Use Rotational Drag=false)
   &CurlNoiseForce(Noise Strength=100, Pan Noise=false)
   &SolveForcesAndVelocity(Clamp Velocity=false, Limit Acceleration=false)
   &ScaleSpriteSize(Scale Factor=&Vector2DFromFloat(Value=&FloatFromCurve(FloatCurve=DataInterface() {propertyValues {Curve {keys [
          {arriveTangent=4.4544, interpMode=RCIM_Cubic, leaveTangent=4.4544, tangentMode=RCTM_User, time=0, value=0}
          <at 0.4741 0.9412 RCIM_Cubic>
          {arriveTangent=-2.1034, interpMode=RCIM_Cubic, leaveTangent=-2.1034, tangentMode=RCTM_User, time=1, value=0}
         ]}}})))
  ]
  rendererClasses [&NiagaraSpriteRendererProperties]
  renderers [{rendererClass=&NiagaraSpriteRendererProperties}]
 }
 FN_SourceParticleRibbon_v3001 {
  bEnabled=false
  emitterSpawnScript [/Game/Effects/Niagara/Modules/Attributes/FN_Source_v2(AttributeReader=DataInterface() {propertyValues {EmitterBinding {bindingMode=Other, emitterName=Fish001}}})]
  emitterUpdateScript [
   &EmitterState()
   &SpawnBurst_Instantaneous(
    Spawn Count=$User.SpawnCount
    Spawn Group=-1
    Spawn Probability=HlslExpression() {hlslExpression="Emitter.LoopCount == 0"}
    Use Spawn Probability=true
   )
   &SpawnRate(
    SpawnRate=/Game/Effects/Niagara/DynamicInputs/Attributes/FN_GetEmitterParticleCount(AttributeReader:NiagaraDataInterfaceParticleRead=$Emitter.AttributeReader, Multiplier:NiagaraFloat=7)
    Use Spawn Probability=false
   )
  ]
  particleSpawnScript [
   /Game/Effects/Niagara/Modules/Attributes/FN_SourceParentID()
   /Game/Effects/Niagara/Modules/Attributes/FN_SourceLocation(Position:Vector3f=/Niagara/DynamicInputs/Transforms/TransformPosition(
     Position=/Game/Effects/Niagara/DynamicInputs/Attributes/FN_GetVectorByID(Attrib name:ENiagaraCommonVectors=::NewEnumerator2, Default on Invalid001:Vector3f=<0 0 0>)
     Source Space=::World
     Destination Space=::Local
    ))
   &InitializeRibbon(
    Color=<rgba 0.2292 0.2292 0.2292 1>
    Lifetime=&MakeCustomFloatFromBool(Bool=$Particles.Locked, True Float=999, False Float=1)
    Module.Write Color=true
    Module.Write Lifetime=true
    Module.Write Mass=true
    Module.Write Material Random=true
    Module.Write Position=false
    Module.Write RibbonTwist=false
    Module.Write RibbonWidth=true
    Ribbon Width=7
   )
   /SetVariables(Particles.DiveThreshold:NiagaraFloat=&FN_FrontEndFloatModifier(Gameplay=0.7, Front End=0.01)) {isSetParameters=true, module=SetVariables_#1}
  ]
  particleUpdateScript [
   &FN_DivingAttributes_v2(Accumulate Dive=0, Accumulate Rest=0.01, SkeletalMesh:NiagaraDataInterfaceSkeletalMesh=$User.SkeletalMesh)
   /Game/Effects/Niagara/Modules/Math/FN_Integrator/FN_Integrator_v2(
    Clamp=true
    DecayRate=HlslExpression() {hlslExpression="Particles.Dive < Particles.DiveThreshold ? Engine.DeltaTime*4 : 0"}
    Input=HlslExpression() {hlslExpression="Particles.Dive > Particles.DiveThreshold ? Engine.DeltaTime*4 : 0"}
    Namespace:ENiagaraNameSpace=::NewEnumerator2
   )
   &GravityForce(Gravity=&Lerp_Vector(
     A=<0 0 50>
     B=&FN_FrontEndVector3Modifier(
      Gameplay=&RotateVectorByQuaternion(
       Vector To Rotate=<-400 0 0>
       Quaternion=/Game/Effects/Niagara/DynamicInputs/Attributes/FN_GetQuaternionByID(Attrib name:ENiagaraCommonQuaternions=::NewEnumerator1, Default:Quat4f=<0 0 0 1>)
      )
      Front End=&TransformVector(Vector=<0 -500 0>, Destination Space=::World)
     )
     Alpha=$Particles.FN_Integrator_v2.Attr
    )) {enabled=false}
   &CurlNoiseForce(Noise Frequency=0.5, Noise Strength=20, Pan Noise=false)
   &SolveForcesAndVelocity(Clamp Velocity=false, Limit Acceleration=false)
   /Game/Effects/Niagara/Modules/Attributes/FN_UpdateSourcedRibbbons(
    Inherit Parent Age=false
    Locked Particles Position:Vector3f=&TransformPosition(
     Position=&FN_GetVectorByID(Attrib name:ENiagaraCommonVectors=::NewEnumerator2, Default on Invalid001:Vector3f=<0 0 0>)
     Source Space=::World
     Destination Space=::Local
    )
    Ribbon Width:NiagaraFloat=$Particles.Initial.RibbonWidth
   ) {module=FN_UpdateLockedParticles}
   &ScaleRibbonWidth(Ribbon Width Scale=&FloatFromCurve(
     FloatCurve=DataInterface() {propertyValues {Curve {keys [
         {arriveTangent=2.7871, interpMode=RCIM_Cubic, leaveTangent=2.7871, tangentMode=RCTM_User, time=-0.0033, value=0.0049}
         {arriveTangent=-0.0276, interpMode=RCIM_Cubic, leaveTangent=-0.0276, tangentMode=RCTM_User, time=0.4759, value=0.7201}
         {interpMode=RCIM_Cubic, tangentMode=RCTM_User, time=1.0011, value=0.002}
        ]}}}
     Scale Curve=2
    ))
   &ScaleColor(Scale Alpha=&Multiply_Float(A=0.1, B=$Particles.RibbonWidth), Scale RGB=<5 5 5>, ScaleA=true, ScaleRGB=false)
   &Material(
    Base Texture Time:NiagaraFloat=&Multiply_Float(A=$Emitter.Age, B=&Lerp_Float(A=-0.4, B=-100, Alpha=$Particles.IntegratedDive))
    PostMult=0.2
    PreMult=0
   )
   /SetVariables() {enabled=false, isSetParameters=true, module=SetVariables_#2}
  ]
  rendererClasses [&NiagaraSpriteRendererProperties, &NiagaraRibbonRendererProperties]
  renderers [
   {rendererClass=&NiagaraSpriteRendererProperties}
   {rendererClass=&NiagaraRibbonRendererProperties, rendererIndex=1}
  ]
 }
 FN_SourceParticleRibbon_v3002 {
  emitterSpawnScript [&FN_Source_v2(AttributeReader=DataInterface() {propertyValues {EmitterBinding {bindingMode=Other, emitterName=Fish001}}})]
  emitterUpdateScript [
   &EmitterState()
   &SpawnBurst_Instantaneous(
    Spawn Count=$User.SpawnCount
    Spawn Group=-1
    Spawn Probability=HlslExpression() {hlslExpression="Emitter.LoopCount == 0"}
    Use Spawn Probability=true
   )
   &SpawnRate(
    SpawnRate=&FN_GetEmitterParticleCount(AttributeReader:NiagaraDataInterfaceParticleRead=$Emitter.AttributeReader, Multiplier:NiagaraFloat=7)
    Use Spawn Probability=false
   )
  ]
  particleSpawnScript [
   &FN_SourceParentID()
   &FN_SourceLocation(Position:Vector3f=&TransformPosition(
     Position=&FN_GetVectorByID(Attrib name:ENiagaraCommonVectors=::NewEnumerator2, Default on Invalid001:Vector3f=<0 0 0>)
     Source Space=::World
     Destination Space=::Local
    ))
   &InitializeRibbon(
    Color=<rgba 0.2292 0.2292 0.2292 1>
    Lifetime=&MakeCustomFloatFromBool(Bool=$Particles.Locked, True Float=999, False Float=1)
    Module.Write Color=true
    Module.Write Lifetime=true
    Module.Write Mass=true
    Module.Write Material Random=true
    Module.Write Position=false
    Module.Write RibbonTwist=false
    Module.Write RibbonWidth=true
    Ribbon Width=7
   )
   /SetVariables(Particles.DiveThreshold:NiagaraFloat=&FN_FrontEndFloatModifier(Gameplay=0.7, Front End=0.01)) {isSetParameters=true, module=SetVariables_#1}
  ]
  particleUpdateScript [
   &FN_DivingAttributes_v2(Accumulate Dive=0, Accumulate Rest=0.01, SkeletalMesh:NiagaraDataInterfaceSkeletalMesh=$User.SkeletalMesh)
   &FN_Integrator_v2(
    Clamp=true
    DecayRate=HlslExpression() {hlslExpression="Particles.Dive < Particles.DiveThreshold ? Engine.DeltaTime*4 : 0"}
    Input=HlslExpression() {hlslExpression="Particles.Dive > Particles.DiveThreshold ? Engine.DeltaTime*4 : 0"}
    Namespace:ENiagaraNameSpace=::NewEnumerator2
   ) {enabled=false}
   &GravityForce(Gravity=&Lerp_Vector(
     A=<0 0 50>
     B=&FN_FrontEndVector3Modifier(
      Gameplay=&RotateVectorByQuaternion(
       Vector To Rotate=<-400 0 0>
       Quaternion=&FN_GetQuaternionByID(Attrib name:ENiagaraCommonQuaternions=::NewEnumerator1, Default:Quat4f=<0 0 0 1>)
      )
      Front End=&TransformVector(Vector=<0 -500 0>, Destination Space=::World)
     )
     Alpha=$Particles.FN_Integrator_v2.Attr
    )) {enabled=false}
   &CurlNoiseForce(Noise Frequency=0.5, Noise Strength=20, Pan Noise=false)
   &SolveForcesAndVelocity(Clamp Velocity=false, Limit Acceleration=false)
   &FN_UpdateSourcedRibbbons(
    Inherit Parent Age=false
    Locked Particles Position:Vector3f=&TransformPosition(
     Position=&FN_GetVectorByID(Attrib name:ENiagaraCommonVectors=::NewEnumerator2, Default on Invalid001:Vector3f=<0 0 0>)
     Source Space=::World
     Destination Space=::Local
    )
    Ribbon Width:NiagaraFloat=$Particles.Initial.RibbonWidth
   ) {module=FN_UpdateLockedParticles}
   &ScaleRibbonWidth(Ribbon Width Scale=&FloatFromCurve(
     FloatCurve=DataInterface() {propertyValues {Curve {keys [
         {arriveTangent=2.7871, interpMode=RCIM_Cubic, leaveTangent=2.7871, tangentMode=RCTM_User, time=-0.0033, value=0.0049}
         {arriveTangent=-0.0276, interpMode=RCIM_Cubic, leaveTangent=-0.0276, tangentMode=RCTM_User, time=0.4759, value=0.7201}
         {interpMode=RCIM_Cubic, tangentMode=RCTM_User, time=1.0011, value=0.002}
        ]}}}
     Scale Curve=2
    ))
   &ScaleColor(Scale Alpha=&Multiply_Float(A=0.1, B=$Particles.RibbonWidth), Scale RGB=<5 5 5>, ScaleA=true, ScaleRGB=false)
   &Material(
    Base Texture Time:NiagaraFloat=&Multiply_Float(A=$Emitter.Age, B=&Lerp_Float(A=-0.4, B=-100, Alpha=$Particles.IntegratedDive))
    PostMult=0.2
    PreMult=0
   )
   /SetVariables() {enabled=false, isSetParameters=true, module=SetVariables_#2}
  ]
  rendererClasses [&NiagaraSpriteRendererProperties, &NiagaraRibbonRendererProperties]
  renderers [
   {rendererClass=&NiagaraSpriteRendererProperties}
   {rendererClass=&NiagaraRibbonRendererProperties, rendererIndex=1}
  ]
 }
 Fish {
  bEnabled=false
  emitterUpdateScript [
   &EmitterState()
   &SpawnBurst_Instantaneous(Spawn Count=$User.SpawnCount, Spawn Probability=HlslExpression() {hlslExpression="Emitter.LoopCount == 0"}, Use Spawn Probability=true)
   &SpawnRate(SpawnRate=10, Use Spawn Probability=false) {enabled=false}
  ]
  particleSpawnScript [
   &InitializeParticle(
    Lifetime=5
    Mesh Scale=&VectorFromFloat(Value=&UniformRangedFloat(Minimum=4, Maximum=5))
    Module.Write Color=true
    Module.Write Lifetime=true
    Module.Write Mass=true
    Module.Write Material Random=true
    Module.Write Position=false
    Module.Write Scale=true
    Module.Write SpriteRotation=false
    Module.Write SpriteSize=true
    Position=&SimulationPosition()
    Sprite Size=<50 50>
   )
   &FN_Orbit_Init_Variables_v2()
   /SetVariables(
    Particles.CurlNoiseSamplePosition:Vector3f=/Niagara/DynamicInputs/Random/RandomVector(
     Vector Scale:NiagaraFloat=500
     Constraint Vector:Vector3f=<0 0 1>
     Randomness Mode:ENiagaraRandomnessMode=::NewEnumerator0
     Random Seed:NiagaraInt32=0
     Fixed Random Seed:NiagaraBool=false
    )
    Particles.ExpansionRate:NiagaraFloat=&UniformRangedFloat(Minimum=2, Maximum=3)
    Particles.IntegratorRate:NiagaraFloat=&UniformRangedFloat(Minimum=0.5, Maximum=1.5)
    Particles.MaterialSpeed:NiagaraFloat=&UniformRangedFloat(Minimum=0.1, Maximum=1.2)
    Particles.MaterialTiling:NiagaraFloat=&UniformRangedFloat(Minimum=0.5, Maximum=1.3)
   ) {isSetParameters=true, module=SetVariables_#3}
   /SetVariables(Particles.DiveThreshold:NiagaraFloat=&FN_FrontEndFloatModifier(Gameplay=0.7, Front End=0.01)) {isSetParameters=true, module=SetVariables_#4}
  ]
  particleUpdateScript [
   &ParticleState(Kill Particles When Lifetime Has Elapsed=false, Let Infinitely Lived Particles Die When Emitter Deactivates=true)
   /SetVariables(Particles.PrevPosition:Vector3f=$Particles.RenderPosition) {isSetParameters=true, module=SetVariables_#5}
   &FN_DivingAttributes_v2(Accumulate Dive=0, Accumulate Rest=0.01, SkeletalMesh:NiagaraDataInterfaceSkeletalMesh=$User.SkeletalMesh) {module=FN_DivingAttributes_v2001}
   &FN_Integrator_v2(
    Clamp=true
    DecayRate=HlslExpression() {hlslExpression="Particles.Dive < Particles.DiveThreshold ? Engine.DeltaTime*4*Particles.IntegratorRate: 0"}
    Input=HlslExpression() {hlslExpression="Particles.Dive > Particles.DiveThreshold ? Engine.DeltaTime*4*Particles.IntegratorRate : 0"}
    Namespace:ENiagaraNameSpace=::NewEnumerator2
   ) {enabled=false}
   &FN_Integrator(
    Clamp=true
    Decrement:NiagaraBool=&SetBoolByFloatComparison(A=$Particles.Dive, B=$Particles.DiveThreshold, Comparison Type=::NewEnumerator6)
    Increment:NiagaraBool=&SetBoolByFloatComparison(A=$Particles.Dive, B=$Particles.DiveThreshold)
    Increment Multiplier=2
   )
   &FN_Integrator_v2(
    Input=HlslExpression() {hlslExpression="Particles.Dive > Particles.DiveThreshold ? Engine.DeltaTime*Particles.IntegratorRate*3 : Engine.DeltaTime"}
    Namespace:ENiagaraNameSpace=::NewEnumerator2
   ) {module=FN_Integrator_v2001}
   &FN_Integrator_v2(
    Input=HlslExpression() {hlslExpression="Particles.Dive > Particles.DiveThreshold ? 0 : Engine.DeltaTime"}
    Namespace:ENiagaraNameSpace=::NewEnumerator2
   ) {module=FN_Integrator_v2002}
   &DynamicMaterialParameters(
    Index 0 Param 1=$Particles.Initial.IntegratorRate
    Index 0 Param 2=&Multiply_Float(A=$Particles.MaterialTiling, B=&Add_Float(A=1, B=0))
    Index 0 Param 3=$Particles.FN_Integrator_v2001.Attr
    Module.Param0WriteEnabled=true
    Module.Param1WriteEnabled=true
    Module.Param2WriteEnabled=true
    Module.Param3WriteEnabled=true
   )
   /SetVariables(Particles.ActivationBlend:NiagaraFloat=&FloatFromCurve(
     FloatCurve=DataInterface() {propertyValues {Curve {keys [
         {arriveTangent=4.6012, interpMode=RCIM_Cubic, leaveTangent=4.6012, tangentMode=RCTM_User, time=-0.0131, value=0.0113}
         {arriveTangent=-0.0489, interpMode=RCIM_Cubic, leaveTangent=-0.0489, tangentMode=RCTM_User, time=0.3615, value=1.1056}
         {arriveTangent=0.0179, interpMode=RCIM_Cubic, leaveTangent=0.0179, tangentMode=RCTM_User, time=0.7898, value=0.9263}
         <at 1.0037 1.0036>
        ]}}}
     CurveIndex=&FN_Fit(
      Clamp:NiagaraBool=true
      Value:NiagaraFloat=&Add_Float(A=$Particles.Age, B=0)
      SourceMin:NiagaraFloat=0
      SourceMax:NiagaraFloat=$Particles.Initial.ExpansionRate
      DestMin:NiagaraFloat=0
      DestMax:NiagaraFloat=1
     )
    )) {isSetParameters=true, module=SetVariables_#6}
   /Game/Effects/Niagara/Modules/Orbit/FN_Orbit_v4(
    Height min/max:Vector2f=&MakeVector2D(
     X=0
     Y=&Sine(Normalized Angle=$Particles.IntegratedDive, Period=&Add_Float(A=1, B=$Particles.MaterialRandom), Scale=20, Bias=0)
    )
    Orbit Center:Vector3f=&Add_Vector(A=Linked() {linkedVariable {name=Engine.Owner.Position, type=NiagaraPosition}}, B=<0 0 150>)
    Radius min/max:Vector2f=&Multiply_Vector2(A=<50 100>, B=&Vector2DFromFloat(Value=$Particles.ActivationBlend))
    Rotation Offset min/max:Vector2f=/Niagara/DynamicInputs/Add/Add_Vector2(
     A:Vector2f=<0 1>
     B:Vector2f=&Vector2DFromFloat(Value=&Multiply_Float(A=$Particles.FN_Integrator_v2002.Attr, B=0.2))
    )
   )
   &CurlNoiseForce(
    Noise Frequency=0.005
    Noise Strength=&Lerp_Float(A=10, B=50, Alpha=$Particles.FN_Integrator_v2.Attr)
    Pan Noise=true
    Sample Position:Vector3f=&Add_Vector(
     A=$Particles.CurlNoiseSamplePosition
     B=&MakeVector(X=0, Y=0, Z=&Multiply_Float(A=$Particles.FN_Integrator_v2001.Attr, B=10))
    )
   )
   /SetVariables(Particles.RenderPosition:Vector3f=&Add_Vector(A=$Particles.FN_Orbit_v4.Position, B=$Transient.PhysicsForce)) {isSetParameters=true, module=SetVariables_#7}
   /SetVariables(Particles.Velocity=&self:Accurate_Velocity()) {isSetParameters=true, module=SetVariables_#8}
   /SetVariables(Particles.RenderOrientation:Quat4f=&LerpQuaternion(
     Quaternion A:Quat4f=&FindQuatBetweenTwoVectors(From Vector=<1 0 0>, To Vector=$Particles.Velocity)
     Quaternion B:Quat4f=$Output.FN_DivingAttributes_v2001.Interpolated Rotation
     Lerp Factor:NiagaraFloat=$Particles.FN_Integrator.Value
     Enable Spherical Linear Interpolation:NiagaraBool=true
    )) {isSetParameters=true, module=SetVariables_#9}
   /SetVariables() {isSetParameters=true}
  ]
  rendererClasses [&NiagaraSpriteRendererProperties, &NiagaraMeshRendererProperties]
  renderers [
   {rendererClass=&NiagaraSpriteRendererProperties}
   {rendererClass=&NiagaraMeshRendererProperties, rendererIndex=1}
   {rendererClass=&NiagaraMeshRendererProperties, rendererIndex=2}
  ]
 }
 Fish001 {
  emitterUpdateScript [
   &EmitterState()
   &SpawnBurst_Instantaneous(Spawn Count=$User.SpawnCount, Spawn Probability=HlslExpression() {hlslExpression="Emitter.LoopCount == 0"}, Use Spawn Probability=true)
  ]
  particleSpawnScript [
   &InitializeParticle(
    Lifetime=5
    Mesh Scale=&VectorFromFloat(Value=&UniformRangedFloat(Minimum=4, Maximum=5))
    Module.Write Color=true
    Module.Write Lifetime=true
    Module.Write Mass=true
    Module.Write Material Random=true
    Module.Write Position=false
    Module.Write Scale=true
    Module.Write SpriteRotation=false
    Module.Write SpriteSize=true
    Position=&SimulationPosition()
    Sprite Size=<50 50>
   )
   &FN_Orbit_Init_Variables_v2()
   /SetVariables(
    Particles.CurlNoiseSamplePosition:Vector3f=&RandomVector(
     Vector Scale:NiagaraFloat=500
     Constraint Vector:Vector3f=<0 0 1>
     Randomness Mode:ENiagaraRandomnessMode=::NewEnumerator0
     Random Seed:NiagaraInt32=0
     Fixed Random Seed:NiagaraBool=false
    )
    Particles.ExpansionRate:NiagaraFloat=&UniformRangedFloat(Minimum=2, Maximum=3)
    Particles.IntegratorRate:NiagaraFloat=&UniformRangedFloat(Minimum=0.5, Maximum=1.5)
    Particles.MaterialSpeed:NiagaraFloat=&UniformRangedFloat(Minimum=0.1, Maximum=1.2)
    Particles.MaterialTiling:NiagaraFloat=&UniformRangedFloat(Minimum=0.5, Maximum=1.3)
   ) {isSetParameters=true, module=SetVariables_#3}
   /SetVariables(Particles.DiveThreshold:NiagaraFloat=&FN_FrontEndFloatModifier(Gameplay=0.7, Front End=0.01)) {isSetParameters=true, module=SetVariables_#4}
  ]
  particleUpdateScript [
   &ParticleState(Kill Particles When Lifetime Has Elapsed=false, Let Infinitely Lived Particles Die When Emitter Deactivates=true)
   /SetVariables(Particles.PrevPosition:Vector3f=$Particles.RenderPosition) {isSetParameters=true, module=SetVariables_#5}
   &FN_DivingAttributes_v2(Accumulate Dive=0, Accumulate Rest=0.01, SkeletalMesh:NiagaraDataInterfaceSkeletalMesh=$User.SkeletalMesh) {module=FN_DivingAttributes_v2001}
   &FN_Integrator(
    Clamp=true
    Decrement:NiagaraBool=&SetBoolByFloatComparison(A=$Particles.Dive, B=$Particles.DiveThreshold, Comparison Type=::NewEnumerator6)
    Decrement Multiplier=3
    Increment:NiagaraBool=&SetBoolByFloatComparison(A=$Particles.Dive, B=$Particles.DiveThreshold)
    Increment Multiplier=3
   )
   &FN_Integrator(
    Decrement:NiagaraBool=&SetBoolByFloatComparison(A=$Particles.Dive, B=$Particles.DiveThreshold)
    Decrement Multiplier=0
    Increment:NiagaraBool=&SetBoolByFloatComparison(A=$Particles.Dive, B=$Particles.DiveThreshold, Comparison Type=::NewEnumerator6)
   ) {module=FN_Integrator002}
   &DynamicMaterialParameters(
    Index 0 Param 1=$Particles.Initial.IntegratorRate
    Index 0 Param 2=&Multiply_Float(A=$Particles.MaterialTiling, B=&Add_Float(A=1, B=0))
    Index 0 Param 3=$Particles.FN_Integrator_v2001.Attr
    Module.Param0WriteEnabled=true
    Module.Param1WriteEnabled=true
    Module.Param2WriteEnabled=true
    Module.Param3WriteEnabled=true
   )
   /SetVariables(Particles.ActivationBlend:NiagaraFloat=&FloatFromCurve(
     FloatCurve=DataInterface() {propertyValues {Curve {keys [
         {arriveTangent=4.6012, interpMode=RCIM_Cubic, leaveTangent=4.6012, tangentMode=RCTM_User, time=-0.0131, value=0.0113}
         {arriveTangent=-0.0489, interpMode=RCIM_Cubic, leaveTangent=-0.0489, tangentMode=RCTM_User, time=0.3615, value=1.1056}
         {arriveTangent=0.0179, interpMode=RCIM_Cubic, leaveTangent=0.0179, tangentMode=RCTM_User, time=0.7898, value=0.9263}
         <at 1.0037 1.0036>
        ]}}}
     CurveIndex=&FN_Fit(
      Clamp:NiagaraBool=true
      Value:NiagaraFloat=&Add_Float(A=$Particles.Age, B=0)
      SourceMin:NiagaraFloat=0
      SourceMax:NiagaraFloat=$Particles.Initial.ExpansionRate
      DestMin:NiagaraFloat=0
      DestMax:NiagaraFloat=1
     )
    )) {isSetParameters=true, module=SetVariables_#6}
   /SetVariables(Particles.OrbitAmount:NiagaraFloat=$Particles.FN_Integrator002.Value) {isSetParameters=true}
   &FN_Orbit_v4(
    Height min/max:Vector2f=&MakeVector2D(
     X=0
     Y=&Sine(Normalized Angle=$Particles.IntegratedDive, Period=&Add_Float(A=1, B=$Particles.MaterialRandom), Scale=20, Bias=0)
    )
    Orbit Center:Vector3f=&Add_Vector(A=Linked() {linkedVariable {name=Engine.Owner.Position, type=NiagaraPosition}}, B=<0 0 150>)
    Radius min/max:Vector2f=&Multiply_Vector2(A=<50 100>, B=&Vector2DFromFloat(Value=$Particles.ActivationBlend))
    Rotation Offset min/max:Vector2f=&Add_Vector2(
     A:Vector2f=<0 1>
     B:Vector2f=&Vector2DFromFloat(Value=&Multiply_Float(A=$Particles.OrbitAmount, B=0.2))
    )
   )
   &CurlNoiseForce(
    Noise Frequency=0.005
    Noise Strength=&Lerp_Float(A=10, B=50, Alpha=$Particles.FN_Integrator_v2.Attr)
    Pan Noise=true
    Sample Position:Vector3f=&Add_Vector(
     A=$Particles.CurlNoiseSamplePosition
     B=&MakeVector(X=0, Y=0, Z=&Multiply_Float(A=$Particles.FN_Integrator_v2001.Attr, B=10))
    )
   )
   /SetVariables(Particles.RenderPosition:Vector3f=&Add_Vector(A=$Particles.FN_Orbit_v4.Position, B=$Transient.PhysicsForce)) {isSetParameters=true, module=SetVariables_#7}
   /SetVariables(Particles.Velocity=&self:Accurate_Velocity()) {isSetParameters=true, module=SetVariables_#8}
   /SetVariables(Particles.RenderOrientation:Quat4f=&LerpQuaternion(
     Quaternion A:Quat4f=&FindQuatBetweenTwoVectors(From Vector=<1 0 0>, To Vector=$Particles.Velocity)
     Quaternion B:Quat4f=&MultiplyQuaternion(
      Quaternion A=$Output.FN_DivingAttributes_v2001.Interpolated Rotation
      Quaternion B=&MakeQuatFromAxisAngle(Axis=<1 0 0>, AngleInDegrees=90)
     )
     Lerp Factor:NiagaraFloat=$Particles.FN_Integrator.Value
     Enable Spherical Linear Interpolation:NiagaraBool=true
    )) {isSetParameters=true, module=SetVariables_#9}
  ]
  rendererClasses [&NiagaraSpriteRendererProperties, &NiagaraMeshRendererProperties]
  renderers [
   {rendererClass=&NiagaraSpriteRendererProperties}
   {rendererClass=&NiagaraMeshRendererProperties, rendererIndex=1}
   {rendererClass=&NiagaraMeshRendererProperties, rendererIndex=2}
  ]
 }
 Fish_3D001 {
  bEnabled=false
  emitterUpdateScript [
   &EmitterScalability(
    Module.bApplySpawnCountScale=false
    Module.bApplySpawnCountScaleByDistanceFraction=false
    Module.bUseMaxDistance=false
    Module.bUseMinDistance=false
   ) {enabled=false}
   &EmitterLifeCycle() {enabled=false}
   &EmitterState()
   &SpawnBurst_Instantaneous(Spawn Count=10, Use Spawn Probability=false) {enabled=false}
   &SpawnRate(SpawnRate=25, Use Spawn Probability=false)
  ]
  particleSpawnScript [
   &FN_DivingAttributes_v2(SkeletalMesh:NiagaraDataInterfaceSkeletalMesh=$User.SkeletalMesh)
   &InitializeParticle(
    Lifetime=2
    Mesh Scale=&VectorFromFloat(Value=&UniformRangedFloat(Minimum=4, Maximum=5))
    Module.Write Color=true
    Module.Write Lifetime=true
    Module.Write Mass=true
    Module.Write Material Random=true
    Module.Write Position=false
    Module.Write Scale=true
    Module.Write SpriteRotation=false
    Module.Write SpriteSize=false
    Position=&SimulationPosition()
   )
   &SystemLocation(MaskSpawnGroup=false, Offset=&UniformRangedVector(Minimum=<0 0 -180>, Maximum=<1 0 -220>)) {enabled=false}
   &DynamicMaterialParameters(
    Index 0 Param 1=&UniformRangedFloat(Minimum=0.1, Maximum=1.2)
    Index 0 Param 2=&UniformRangedFloat(Minimum=0.5, Maximum=1.3)
    Module.Param0WriteEnabled=true
    Module.Param1WriteEnabled=true
    Module.Param2WriteEnabled=true
    Module.Param3WriteEnabled=true
   )
   &CylinderLocation(
    Cylinder Height=10
    Cylinder Radius=&Lerp_Float(A=50, B=25, Alpha=$Particles.Dive)
    Lathe Profile=DataInterface() {propertyValues {Curve {keys [<at 0 0.451 RCIM_Cubic>, {arriveTangent=0.049, interpMode=RCIM_Cubic, leaveTangent=0.049, time=0.5, value=1}, <at 1 0.5 RCIM_Cubic>]}}}
    MaskSpawnGroup=false
    Module.Enable Lathe Profile=false
    Module.Override Randomness=false
    Module.Surface Only=true
    Offset=&Lerp_Vector(A=<0 0 125>, B=<0 0 270>, Alpha=$Particles.Dive)
    Surface Only Band Thickness=30
    Use Endcaps In Surface Only Mode=false
   )
   &CurlNoiseForce(Noise Strength=25, Pan Noise=false)
   &ApplyInitialForces()
  ]
  particleUpdateScript [
   &RotateAroundPoint(Radius=100, Rotation Phase:NiagaraFloat=&UniformRangedFloat(Minimum=0, Maximum=1), Rotation Rate=500) {enabled=false}
   &FN_DivingAttributes_v2(SkeletalMesh:NiagaraDataInterfaceSkeletalMesh=$User.SkeletalMesh) {module=FN_DivingAttributes_v2001}
   &VortexVelocity(Module.Use Influence Falloff=false, Velocity Amount:NiagaraFloat=&Lerp_Float(A=200, B=0, Alpha=$Particles.Dive))
   &AddVelocity(Velocity=&Lerp_Vector(A=<0 0 0>, B=<0 20 -100>, Alpha=$Particles.Dive))
   &SolveForcesAndVelocity(Clamp Velocity=false, Limit Acceleration=false)
   &ParticleState()
   &Color() {enabled=false}
   &ScaleMeshSize(Scale Factor=&VectorFromFloat(Value=&FloatFromCurve(FloatCurve=DataInterface() {propertyValues {Curve {keys [<at -0.0046 0>,<at 0.5 1>,<at 1.0011 0.002>]}}})))
   &GenerateLocationEvent(Event Send Rate=2, Module.Use Event Send Rate=false)
  ]
  rendererClasses [&NiagaraMeshRendererProperties]
  renderers [{rendererClass=&NiagaraMeshRendererProperties, rendererIndex=0}]
 }
}
summary {
 systemName=NS_Contrail_Fish_School
 emitters [
  {bEnabled=false, emitterName=DEBUG, rendererClasses [&NiagaraMeshRendererProperties]}
  {bEnabled=false, emitterName=Fish_3D001, rendererClasses [&NiagaraMeshRendererProperties]}
  {bEnabled=false, emitterName=FN_Contrails_Bubbles_Child, rendererClasses [&NiagaraSpriteRendererProperties]}
  {emitterName=FN_Contrails_Bubbles, simTarget=GPUComputeSim, rendererClasses [&NiagaraSpriteRendererProperties]}
  {
   bEnabled=false
   emitterName=Fish
   rendererClasses [&NiagaraSpriteRendererProperties, &NiagaraMeshRendererProperties]
  }
  {emitterName=FN_Contrails_Bubble_Low, rendererClasses [&NiagaraSpriteRendererProperties]}
  {emitterName=Fish001, rendererClasses [&NiagaraSpriteRendererProperties, &NiagaraMeshRendererProperties]}
  {
   bEnabled=false
   emitterName=FN_SourceParticleRibbon_v3001
   rendererClasses [&NiagaraSpriteRendererProperties, &NiagaraRibbonRendererProperties]
  }
  {
   emitterName=FN_SourceParticleRibbon_v3002
   rendererClasses [&NiagaraSpriteRendererProperties, &NiagaraRibbonRendererProperties]
  }
 ]
 userVariables [
  User.bIsFrontEnd {type=NiagaraBool}
  User.bIsFrontEndPreview {type=NiagaraBool}
  User.Orbit Center {type=Vector3f, description="", defaultValue {struct=Vector3f, value=<0 0 0>}}
  User.PlayerSpeed {type=NiagaraFloat, defaultValue=0, description=""}
  User.RibbonIDOffset {type=NiagaraInt32}
  User.SkeletalMesh {
   type=NiagaraDataInterfaceSkeletalMesh
   defaultValue {
    struct=NiagaraExt_VariableValue_DataInterface
    value {dataInterface=&self:NiagaraDataInterfaceSkeletalMesh_0}
   }
  }
  User.SpawnCount {type=NiagaraInt32, defaultValue=25}
 ]
}
